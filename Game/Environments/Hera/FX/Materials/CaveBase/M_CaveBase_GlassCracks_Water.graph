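# Two-sided masked lit material for water at cracks in breaking glass, previewed on the Hera CaveBase water tank crack mesh.
# Random dots pan along UV1 at vertex-red-varied speed, inside a UV2 sphere mask sized by vertex red minus GlassCracking.
# That mask multiplies with an end-faded V gradient, static dots and cell noise, then feeds a Bayer dither.
# Side branches build a flattened cell-noise normal and Fresnel-tinted base and emissive colors.
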Material /Game/Environments/Hera/FX/Materials/CaveBase/M_CaveBase_GlassCracks_Water
graph {
 properties {
  bAutomaticallySetUsageInEditor=false
  blendMode=BLEND_Masked
  previewMesh=/Game/Environments/Hera/FX/Meshes/SM_WaterTankBreak_GlassCracks_01
  translucencyLightingMode=TLM_VolumetricPerVertexNonDirectional
  twoSided=true
  usedShadingModels=Default Lit
 }
 nodes [
  ComponentMask(None=TextureCoordinate() {coordinateIndex=2}) {g=true}
  n1 = Multiply(
   A=ComponentMask(None=VectorParameter() {parameterName=WaterTiling, defaultValue=<rgba 25 1 0 1>}.RGB) {g=true, r=true}
   B=TextureCoordinate() {coordinateIndex=1}
  )
  n2 = Time() {period=2}
  n3 = Multiply(
   A=LinearInterpolate(Alpha=VertexColor().R) {constA=0.7}
   B=ComponentMask(None=VectorParameter() {parameterName=Velocity, defaultValue=<rgba 0 -1 0 1>}.RGB) {g=true, r=true}
  )
  n4 = TextureSample(UVs=Power(Base=Panner(Coordinate=@n1, Time=@n2, Speed=@n3) {speedY=-1}) {constExponent=1}) {samplerType=SAMPLERTYPE_LinearColor, texture=/Game/Effects/Fort_Effects/Textures/Tile/T_Random_Dots}
  n5 = ScalarParameter() {defaultValue=0.25, parameterName=GlassCracking}
  n6 = Subtract(A=VertexColor().R, B=@n5)
  n7 = Multiply(
   A=Power(Base=Multiply(A=@n4.B), Exp=@n5) {constExponent=1}
   B=SphereMask(A=TextureCoordinate() {coordinateIndex=2}, B=Constant2Vector() {g=0.5, r=0.5}, Radius=Multiply(A=@n6) {constB=4}) {attenuationRadius=0.2, hardnessPercent=0}
  )
  n8 = ComponentMask(None=MaterialFunctionCall(UV Channel=TextureCoordinate() {coordinateIndex=1}) {materialFunction=&LinearGradient}.VGradient) {g=true}
  n9 = OneMinus(None=@n8)
  n10 = TextureCoordinate() {uTiling=8, vTiling=8}
  n11 = TextureSample(UVs=Add(A=@n10, B=Constant2Vector() {r=0.1})) {samplerType=SAMPLERTYPE_LinearColor, texture=&T_Random_Dots}
  n12 = OneMinus(None=@n11.B)
  n13 = Multiply(
   A=Multiply(A=ScalarParameter() {defaultValue=1, parameterName=Opacity Mask}, B=Multiply(A=@n8, B=@n9))
   B=Saturate(None=Power(Base=@n12) {constExponent=8000})
  )
  n14 = Multiply(A=TextureSample(UVs=TextureCoordinate()) {samplerType=SAMPLERTYPE_LinearGrayscale, texture=/Game/Effects/Fort_Effects/Textures/Tile/T_CellNoise_Tile_M}.G)
  MaterialFunctionCall!(In=Multiply(A=Multiply(A=@n7, B=@n13), B=Power(Base=@n14) {constExponent=25}), Iterations, Scale) {materialFunction=/Game/Packages/Fortress_SharedMaterials/Base_Material_Functions/MF_DitherBayerNoTex}
  n15 = TextureSample(UVs=@n10) {samplerType=SAMPLERTYPE_Normal, texture=/Game/Effects/Fort_Effects/Textures/Tile/T_CellNoise_Tile_N}
  MaterialFunctionCall(Normal=@n15.RGB, Flatness=ScalarParameter() {defaultValue=1, parameterName="Flatness (S)"}) {materialFunction=&FlattenNormal}
  Multiply()
  Multiply(A=@n15.R)
  n16 = Fresnel()
  Multiply(A=VectorParameter() {parameterName=Base Color, defaultValue=<rgba 0.1 0.1 0.1 0>}.RGB, B=@n16)
  Multiply(A=Multiply(A=@n16, B=VectorParameter() {parameterName=Emissive Color}.RGB), B=ComponentMask(None=@n11.R) {r=true})
  ScalarParameter()
  ScalarParameter() {parameterName="Contrast (S)"}
  SphereMask()
 ]
}
