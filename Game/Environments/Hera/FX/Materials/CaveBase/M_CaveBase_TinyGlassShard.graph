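# Masked, two-sided lit material for Niagara sprite particles, named for tiny glass shards in the Hera CaveBase effects, using a 2x2 rock-chunk texture.
# Particle alpha times the texture's green channel feeds a Bayer dither, standing in for soft opacity with responsive AA enabled.
# Shadow-tinted particle color is darkened by a normal-map dot with sunlight, gated by an Is Translucent blend.
# A PreviousFrameSwitch offsets the previous frame along particle direction by a dynamic motion blur strength.
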Material /Game/Environments/Hera/FX/Materials/CaveBase/M_CaveBase_TinyGlassShard
graph {
 properties {
  bEnableResponsiveAA=true
  bTangentSpaceNormal=false
  bUsedWithNiagaraSprites=true
  bUsedWithStaticMesh=false
  blendMode=BLEND_Masked
  pixelDepthOffsetMode=PDOM_Legacy
  twoSided=true
  usedShadingModels=Default Lit
 }
 nodes [
  n1 = ParticleColor()
  MaterialFunctionCall!(
   In=Multiply(A=@n1.A, B=TextureSampleParameter2D() {parameterName=Color and Alpha, texture=/Game/Effects/Fort_Effects/Textures/Rocks/T_Chunks_2x2_D}.G)
   Iterations
   Scale
  ) {materialFunction=/Game/Packages/Fortress_SharedMaterials/Base_Material_Functions/MF_DitherBayerNoTex}
  n2 = Multiply(A=VectorParameter() {parameterName=ShadowColor, defaultValue=<rgba 0.1 0.1 0.1 1>}.RGB, B=@n1.RGB)
  n3 = Multiply(A=@n2) {constB=0.75}
  n4 = CollectionParameter() {
   collection=&FortniteMaterialParameters
   parameterName=SunlightDirectionalVector
  }
  n5 = DotProduct(
   A=ComponentMask(None=@n4) {b=true, g=true, r=true}
   B=TextureSampleParameter2D() {parameterName=Normal, samplerType=SAMPLERTYPE_Normal, texture=/Game/Effects/Fort_Effects/Textures/Rocks/T_Chunks_2x2_N}.RGB
  )
  Multiply(
   A=LinearInterpolate(A=@n2, B=LinearInterpolate(A=@n2, B=@n3, Alpha=@n5), Alpha=ScalarParameter() {parameterName=Is Translucent})
   B=ScalarParameter() {defaultValue=1, parameterName=Fake Lighting}
  )
  ParticleColor()
  n6 = Multiply(A=VertexNormalWS())
  n7 = ParticleDirection()
  n8 = DynamicParameter() {paramNames [MotionBlur_Strength, Param2, Param3, Param4]}
  PreviousFrameSwitch(Current Frame=@n6, Previous Frame=Subtract(A=@n6, B=Multiply(A=@n7, B=@n8.MotionBlur_Strength)))
  ScalarParameter() {defaultValue=0.7, parameterName=Roughness}
 ]
}
