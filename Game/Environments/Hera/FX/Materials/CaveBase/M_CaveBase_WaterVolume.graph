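# Masked, lit water volume material for the Hera cave base effects, flagged for Niagara mesh particles.
# A panning normal map distorts the UVs of scrolling foam noise, which is multiplied by particle color and by particle alpha with an inverted Fresnel.
# Further branches build depth-fade-scaled refraction and vertex-normal offset terms.
# The foam UVs reuse the U scale and speed on both axes, leaving the V channels unused.
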Material /Game/Environments/Hera/FX/Materials/CaveBase/M_CaveBase_WaterVolume
graph {
 properties {
  bAutomaticallySetUsageInEditor=false
  bUsedWithNiagaraMeshParticles=true
  blendMode=BLEND_Masked
  refractionMethod=RM_2DOffset
  translucencyLightingMode=TLM_VolumetricPerVertexDirectional
  usedShadingModels=Default Lit
 }
 nodes [
  Constant() {r=0.9}
  n1 = ParticleColor()
  n2 = TextureCoordinate() {uTiling=6}
  n3 = ScalarParameter() {defaultValue=1, parameterName=Noise Scale}
  n4 = Multiply(A=ScalarParameter() {defaultValue=0.2, parameterName=Noise Pan U}, B=Time())
  n5 = Constant()
  n6 = TextureSample(UVs=Add(A=Multiply(A=@n2, B=@n3), B=AppendVector(A=@n4, B=@n5))) {samplerType=SAMPLERTYPE_Normal, texture=/Game/Global/Textures/Default/Tileable_Patterns/2ndInfectedMapNormal}
  n7 = ComponentMask(None=Subtract(A=NamedRerouteDeclaration(None=@n6.RGB) {name=Normal Map, nodeColor=<rgba 0.3828 0 1 1>}) {constB=0.5}) {g=true, r=true}
  n8 = ScalarParameter() {defaultValue=0.2, group=Foam, parameterName=Foam UV Distortion}
  n9 = VectorParameter() {group=Foam, parameterName=Foam_UV, channelNames {a=Pan Speed V, b=Pan Speed U, g=V Scale, r=U Scale}, defaultValue=<rgba 3 1 -0.3 0.4>}
  n10 = Multiply(A=TextureCoordinate(), B=AppendVector(A=@n9.U Scale, B=@n9.U Scale))
  n11 = Multiply(A=AppendVector(A=@n9.Pan Speed U, B=@n9.Pan Speed U), B=Time())
  n12 = TextureSample(UVs=Add(A=Multiply(A=@n7, B=@n8), B=Add(A=@n10, B=@n11))) {texture=/Game/Effects/Fort_Effects/Textures/Fire/HF_FireNoisesPacked}
  Multiply(A=Multiply(A=@n1.A, B=@n12.G), B=OneMinus(None=Fresnel() {exponent=0.25}))
  Multiply(A=@n1.RGB, B=@n12.G)
  n13 = NamedRerouteUsage()
  Multiply(
   A=Multiply(A=@n13, B=ScalarParameter() {defaultValue=1, parameterName=Refract Strength})
   B=LinearInterpolate(Alpha=DepthFade() {fadeDistanceDefault=250}) {constA=1, constB=1.2}
  )
  Multiply(A=VertexNormalWS(), B=Multiply(A=ScalarParameter() {parameterName=MultWPO}, B=@n13))
 ]
}
